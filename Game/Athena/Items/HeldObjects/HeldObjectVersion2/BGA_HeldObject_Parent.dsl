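; Parent Blueprint for Fortnite held objects that players pick up, place or throw, with overridable interaction hooks and server-side pickup on second interact.
; When the projectile stops, the server can align it to walls and ceilings by nearest axis, or attach it to vehicles.
; Rep notifies drive attachment, hiding and movement restart when the supporting actor is destroyed.
; Replicated gameplay cues cover hits, throws, pickups, water entry and death; optional underwater spawns expire.
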(event EventBeginPlay
  (bind (_returnvalue_x _returnvalue_y _returnvalue_z) (Math|Vector|BreakVector (Transformation|GetActorLocation)))
  ;[1 1 1] Begin Play - Set attach default transform and save off collision setting
  (|CallParentFunction)
  (Networking|FlushNetDormancy)
  (CallFunction|SetAttachDefaultTransform)
  (Variables|Default|SetCapsuleDefaultCollisionResponseToPawn (Collision|GetCollisionResponsetoChannel (Variables|Default|GetCapsule) "ECC_Pawn"))
  ;[1 1 1] Wait for Playlist ready, check if we want to kill this object if its underwater. If so, check that it IS underwater and then set lifespan
  (Class|FortAsyncActionCurrentPlaylistReady|PlaylistDataReadyAsync
    (:OnReady
      (if (Variables|Default|GetShouldKillWhenSpawningUnderwater)
        (if (or (Fortnite|WaterHelpers|IsActorSubmergedinOcean self) (<= _returnvalue_z (Fortnite|WaterHelpers|GetOceanTotalHeight)))
          (Actor|SetLifeSpan (Math|Random|RandomFloatinRange 1.0 1e+01)))))))

(event Custom|HideActor
  ;[1 1 1] Hide Actor - Custom Event that can be called on children to do custom changes for when the actor is hidden with the rep notify.
  (Networking|FlushNetDormancy)
  (Rendering|SetActorHiddenInGame true))

(event Interaction|EventBlueprintOnInteract
  (params (InteractingPawn :type Fort%20Pawn%20Object%20Reference) (InteractionBeingAttempted :type EInteractionBeingAttempted%20Enum :default "FirstInteraction"))
  ;[1 1 1] On Interact - Check if the interact is first or second interaction. Execute events that can be overridden in children and by default pick up the item on second interact.
  (if (Variables|Interact|GetRestoreDefaultPawnCollisionOnInteract)
    (Collision|SetCollisionResponsetoChannel (Variables|Default|GetCapsule) "ECC_Pawn" (Variables|Default|GetCapsuleDefaultCollisionResponseToPawn)) ; Adjust collision so we don't prevent pawns from riding in the vehicle
    (else
      _))
  (Networking|FlushNetDormancy)
  (switch Utilities|FlowControl|Switch|SwitchonEInteractionBeingAttempted InteractionBeingAttempted
    (:FirstInteraction
      (CallFunction|OnFirstInteract InteractingPawn))
    (:SecondInteraction
      (CallFunction|OnSecondInteract InteractingPawn)
      (bind _asfort_player_pawn (Utilities|Casting|CastToFortPlayerPawn InteractingPawn))
      (if (Networking|IsServer)
        (HeldObjects|Behavior|PickupHeldObject (Variables|Default|GetFortHeldObject) _asfort_player_pawn)))))

(event Custom|OnFirstInteract
  (params (Interacting%20Pawn :type Fort%20Pawn%20Object%20Reference)))

(event Custom|OnSecondInteract
  (params (InteractingPawn :type Fort%20Pawn%20Object%20Reference)))

(event OnLinkedActorDestroyed(FortLinkToActor)
  (params (DamageCauser :type Actor%20Object%20Reference))
  ;[1 1 1] On Linked Actor Destroyed - Reinit movement when base is destroyed by incrementing a rep notify that sets tells the projectile movement component to begin simulating
  (Networking|FlushNetDormancy)
  (Variables|ProjectileMovement|SetRepStartMoving 1))

(event OnHeldObjectPlaced(FortHeldObject)
  ;[1 1 1] On Held Object Placed - Check that the object has been placed so that attachment on walls works correctly
  (Networking|FlushNetDormancy)
  (CallFunction|SetBeenThrownPlaced)
  (CallFunction|UbindPlayerControlleronUnPosses))

(event Custom|OnVehicleDestroyed
  (params (DestroyedActor :type Actor%20Object%20Reference))
  ;[1 1 1] OnVehicleDestroyed
  (Networking|SwitchHasAuthority
    (:Authority
      (if (Utilities|Equal(Object) DestroyedActor (Variables|Default|GetVehicleAttached))
        (Networking|FlushNetDormancy)
        (Variables|ProjectileMovement|SetRepStartMoving 1)
        (if (Variables|Default|GetIsAttachedToVehicle)
          (Variables|Default|SetIsAttachedToVehicle false))))))

(event EventDestroyed
  ;[1 1 1] Destroyed - If destroy event is called and we have never been thrown or placed play generic death. This is for things like swapping items on first select
  (Networking|FlushNetDormancy)
  (if (not (Variables|ProjectileMovement|GetEverBeenThrownPlaced))
    (CallFunction|PlayGenericDeath)))

(event Collision|EventHit
  (params (MyComp :type Primitive%20Component%20Object%20Reference) (Other :type Actor%20Object%20Reference) (OtherComp :type Primitive%20Component%20Object%20Reference) (bSelfMoved :type Boolean) (HitLocation :type Vector :default "0, 0, 0") (HitNormal :type Vector :default "0, 0, 0") (NormalImpulse :type Vector :default "0, 0, 0") (Hit :type Hit%20Result%20Structure%20%28by%20ref%29))
  (bind _objectweights (Variables|GameplayCues|GetObjectWeights))
  (bind _returnvalue (Math|Conversions|ToInteger(Byte) _objectweights))
  (bind _returnvalue_2 (Ability|GameplayCue|MakeGameplayCueParameters :NormalizedMagnitude 0.0 :RawMagnitude 0.0 :Location HitLocation :Normal HitNormal :GameplayEffectLevel 1 :AbilityLevel _returnvalue :bReplicateLocationWhenUsingMinimalRepProxy false))
  ;[1 1 1 0.5 40] Execute Gameplay Cues
  ;;[1 1 1] Hit - Send Cues on hit events for the world and player
  (Networking|FlushNetDormancy)
  ;;;[1 1 1] Make sure not in water
  (if (not (Water|IsinWater (Variables|Default|GetFortWaterInteraction)))
    ;;;[1 1 1] Make sure we are moving some
    (if (>= (Math|Vector|VectorLength (Variables|/Script/Engine.MovementComponent|GetVelocity (Variables|Default|GetFortProjectileMovement))) (Variables|Default|GetHit%20Velocity%20Threshold))
      (Utilities|Casting|CastToFortPawn Other
        (:then
          (Ability|ExecuteGameplayCueWithParamsReplicated (Variables|GameplayCues|GetGC_HitPlayer) _returnvalue_2))
        (:CastFailed
          (Utilities|FlowControl|DoOnce false
            (:Completed
              (Ability|ExecuteGameplayCueWithParamsReplicated (Variables|GameplayCues|GetGC_HitWorld) _returnvalue_2)
              (Utilities|FlowControl|Delay 0.25))) ; Fix the issue where hit sometimes triggers 5x in a frame
        )))))

(event Custom|PlayGenericDeath
  ;[1 1 1 0.5 40] Execute Gameplay Cues
  ;;[1 1 1] Play Generic Death - Plays generic death GC used for things that explode on death but should play this death instead when killing from picking it up and breaking it.
  (Ability|ExecuteGameplayCueWithParamsReplicated (Variables|GameplayCues|GetGC_GenericDeath) (Ability|GameplayCue|MakeGameplayCueParameters :NormalizedMagnitude 0.0 :RawMagnitude 0.0 :Location (Transformation|GetActorLocation) :Normal (Transformation|GetActorForwardVector) :GameplayEffectLevel 1 :AbilityLevel (Math|Conversions|ToInteger(Byte) (Variables|GameplayCues|GetObjectWeights)) :bReplicateLocationWhenUsingMinimalRepProxy false))
  (CallFunction|HideActor)
  (if (Networking|IsServer)
    (Variables|Death|SetRepHideActor true)))

(event OnHeldObjectThrown(FortHeldObject)
  ;[1 1 1 0.5 40] Execute Gameplay Cues
  ;;[1 1 1] On Held Object Thrown - Send cue for Throw
  (Ability|ExecuteGameplayCueWithParamsReplicated (Variables|GameplayCues|GetGC_Throw) (Ability|GameplayCue|MakeGameplayCueParameters :NormalizedMagnitude 0.0 :RawMagnitude 0.0 :Location (Transformation|GetActorLocation) :Normal (Transformation|GetActorForwardVector) :GameplayEffectLevel 1 :AbilityLevel (Math|Conversions|ToInteger(Byte) (Variables|GameplayCues|GetObjectWeights)) :bReplicateLocationWhenUsingMinimalRepProxy false))
  (CallFunction|SetBeenThrownPlaced)
  (Quest|SendComplexCustomStatEvent (Quests|GetQuestManager (Fortnite|GetFortPlayerControllerfromActor (HeldObjects|Instigator|GetInstigator (Variables|Default|GetFortHeldObject))) "Invalid") 0 "(GameplayTags=)" (Variables|Default|GetThrownQuestCreditTargetTag) 1)
  (CallFunction|UbindPlayerControlleronUnPosses))

(event EventOnDeathServer
  (params (Damage :type Float%20%28single-precision%29) (DamageTags :type Gameplay%20Tag%20Container%20Structure%20%28by%20ref%29) (Momentum :type Vector :default "0, 0, 0") (HitInfo :type Hit%20Result%20Structure%20%28by%20ref%29) (InstigatedBy :type Controller%20Object%20Reference) (DamageCauser :type Actor%20Object%20Reference) (EffectContext :type Gameplay%20Effect%20Context%20Handle%20Structure))
  ;[1 1 1 0.5 40] Execute Gameplay Cues
  ;;[1 1 1] On Death Server - Play Death GC and then hide actor if settings have not been overridden.
  (Networking|FlushNetDormancy)
  (if (Variables|GameplayCues|GetPlayDeathGC)
    (Ability|ExecuteGameplayCueWithParamsReplicated (Variables|GameplayCues|GetGC_Death) (Ability|GameplayCue|MakeGameplayCueParameters :NormalizedMagnitude 0.0 :RawMagnitude 0.0 :Location (Transformation|GetActorLocation) :Normal (Transformation|GetActorForwardVector) :GameplayEffectLevel 1 :AbilityLevel (Math|Conversions|ToInteger(Byte) (Variables|GameplayCues|GetObjectWeights)) :bReplicateLocationWhenUsingMinimalRepProxy false))
    (if (Variables|Death|GetSetHideActorOnDeath)
      (Variables|Death|SetRepHideActor true))))

(event OnEnterWaterEvent(FortWaterInteraction)
  (params (WaterBody :type Fort%20Water%20Body%20Actor%20Object%20Reference) (WaterInteractionComponent :type Fort%20Water%20Interaction%20Component%20Object%20Reference) (bIsFirstBody :type Boolean))
  (bind (_waterplanelocation _waterplanenormal _watersurfaceposition _waterdepth _waterbodyidx) (FX|GetLastWaterSurfaceInfo (Variables|Default|GetFortWaterInteraction) :!class /Script/FortniteGame.FortWaterInteractionComponent))
  ;[1 1 1 0.5 40] Execute Gameplay Cues
  ;;[1 1 1] On Enter Water - Play Splash Cues
  (bind _returnvalue (Fortnite|HeldObjects|GetHeldObjectAttachedtoPlayer (HeldObjects|Instigator|GetInstigator (Variables|Default|GetFortHeldObject))))
  (if (not (Utilities|Equal(Object) _returnvalue self))
    (Ability|ExecuteGameplayCueWithParamsReplicated (Variables|GameplayCues|GetGC_EnterWater) (Ability|GameplayCue|MakeGameplayCueParameters :NormalizedMagnitude (Math|Conversions|ToFloat(Integer) _waterbodyidx) :RawMagnitude _waterdepth :Location _waterplanelocation :Normal "0,0,0" :GameplayEffectLevel 1 :AbilityLevel (Math|Conversions|ToInteger(Byte) (Variables|GameplayCues|GetObjectWeights)) :bReplicateLocationWhenUsingMinimalRepProxy false))))

(event OnHeldObjectPickedUp(FortHeldObject)
  ;[1 1 1 0.5 40] Execute Gameplay Cues
  ;;[1 1 1] On Pickup - call the reset relative transform event
  (Utilities|FlowControl|Sequence
    (:then_0
      (Networking|FlushNetDormancy)
      (if (Variables|ProjectileMovement|GetEverBeenThrownPlaced)
        (Ability|ExecuteGameplayCueWithParamsReplicated (Variables|GameplayCues|GetGC_Pickup) (Ability|GameplayCue|MakeGameplayCueParameters :NormalizedMagnitude 0.0 :RawMagnitude 0.0 :Location (Transformation|GetActorLocation) :Normal (Transformation|GetActorForwardVector) :GameplayEffectLevel 1 :AbilityLevel (Math|Conversions|ToInteger(Byte) (Variables|GameplayCues|GetObjectWeights)) :bReplicateLocationWhenUsingMinimalRepProxy false))
        (else
          _))
      (if (Variables|ProjectileMovement|GetAttachToWallsAndCeilings)
        (CallFunction|ResetRelativeTransformonAttach)))
    (:then_1
      (if (Variables|Default|GetIsAttachedToVehicle)
        (Variables|Default|SetIsAttachedToVehicle false)))
    (:then_2
      (CallFunction|BindPlayerControlleronUnPosses))))

(event OnProjectileBounce(FortProjectileMovement)
  (params (ImpactResult :type Hit%20Result%20Structure%20%28by%20ref%29) (ImpactVelocity :type Vector%20%28by%20ref%29 :default "0, 0, 0"))
  (bind (_bblockinghit _binitialoverlap _time _distance _location _impactpoint _normal _impactnormal) (Collision|BreakHitResult ImpactResult))
  (bind (_bblockinghit_2 _binitialoverlap_2 _time_2 _distance_2 _location_2 _impactpoint_2 _normal_2 _impactnormal_2 _physmat _hitactor _hitcomponent) (Collision|BreakHitResult ImpactResult))
  (bind (_asfort_player_pawn _bsuccess) (Utilities|Casting|CastToFortPlayerPawn _hitactor))
  (bind (_asfort_pawn _bsuccess_2) (Utilities|Casting|CastToFortPawn _hitactor))
  (bind (_asfort_vehicle_interface _bsuccess_3) (Utilities|Casting|CastToFortVehicleInterface _hitactor))
  ;[1 1 1] On Projectile Bounce - Check if we hit a flat enough surface, if so stop.
  (if (Variables|ProjectileMovement|GetStopOnFlatSurface)
    (if (>= (.z _impactnormal) (select (and (Variables|ProjectileMovement|GetAttachToWallsAndCeilings) (Variables|ProjectileMovement|GetEverBeenThrownPlaced)) -1e+01 (Variables|ProjectileMovement|GetWorldStopSlop)))
      (if (Math|Boolean|ORBoolean _bsuccess (Actor|ActorHasTag _hitactor "NoStick") (Components|ComponentHasTag _hitcomponent "NoStick"))
        (if (and (not _bsuccess_3) (not _bsuccess_2))
          ;[1 1 1] If you hit a target that isn't a valid stop adjust velocity.
          (Variables|/Script/Engine.MovementComponent|SetVelocity (Math|Vector|vector+vector (Math|Vector|MirrorVectorbyNormal (* (Variables|/Script/Engine.MovementComponent|GetVelocity (Variables|Default|GetFortProjectileMovement)) 0.75) "0, 0,1.000000") (Math|Vector|MakeVector 0.0 0.0 (* (Math|Vector|VectorLength (Variables|/Script/Engine.MovementComponent|GetVelocity (Variables|Default|GetFortProjectileMovement))) 0.25))) (Variables|Default|GetFortProjectileMovement)))
        (else
          (Game|Components|ProjectileMovement|StopSimulating (Variables|Default|GetFortProjectileMovement) ImpactResult))))))

(event OnProjectileStop(FortProjectileMovement)
  (params (ImpactResult :type Hit%20Result%20Structure%20%28by%20ref%29))
  (bind (_bblockinghit _binitialoverlap _time _distance _location _impactpoint _normal _impactnormal _physmat _hitactor _hitcomponent) (Collision|BreakHitResult ImpactResult))
  (bind (_bblockinghit_2 _binitialoverlap_2 _time_2 _distance_2 _location_2 _impactpoint_2 _normal_2 _impactnormal_2 _physmat_2 _hitactor_2) (Collision|BreakHitResult ImpactResult))
  (bind (_asfort_vehicle_interface _bsuccess) (Utilities|Casting|CastToFortVehicleInterface _hitactor_2))
  ;[1 1 1 0.5 40] Handle attachment
  ;;[1 1 1] On Projectile Stop - If we have AttachToWallsAndCielings then find the best attach angle and use that
  (Networking|FlushNetDormancy)
  (if (Networking|IsServer)
    (Variables|ProjectileMovement|SetRepStartMoving 0)
    (if (Variables|ProjectileMovement|GetAutoEnableInteractionsOnStop)
      _
      (else
        _))
    (if (not _bsuccess)
      (if (and (Variables|ProjectileMovement|GetAttachToWallsAndCeilings) (Variables|ProjectileMovement|GetEverBeenThrownPlaced))
        (CallFunction|FindAttachAngle _normal)
        (Utilities|FlowControl|Delay 0.06)
        (LinkToActorComponent|PerformLinkingCheck (Variables|Default|GetFortLinkToActor) (Variables|ProjectileMovement|GetDirectionToAttach) (Variables|ProjectileMovement|GetEmptyAttachLinkActor))
        (CallFunction|AttachSetup _location _normal))
      (else
        (Utilities|FlowControl|Delay 0.1)
        (CallFunction|ServerOnlyHandleAttachtoVehicle _normal _hitcomponent)))))

(event Custom|AttachSetup
  (params (Location :type Vector%20%28by%20ref%29) (HitNormal :type Vector%20%28by%20ref%29) (NewParam :type Boolean))
  ;[1 1 1 0.5 40] Handle attachment
  ;;[1 1 1] Attach Setup - make a rotation from the hit normal and then make a transform. Setting a random scale even though we dont use it to ensure the rep notify fires.
  (Variables|ProjectileMovement|SetAttachTransform (Math|Transform|MakeTransform Location (Math|Rotator|MakeRotfromXZ (Math|Vector|RotateVectorAroundAxis (Transformation|GetActorForwardVector) (* (Class|AbilityFunctions|GetAngleBetweenTwoVectors "0, 0,1.000000" HitNormal) -1.0) (Math|Vector|CrossProduct HitNormal "0, 0,1.000000")) HitNormal) (Math|Conversions|ToVector(Float) (Math|Random|RandomFloat)))))

(event Custom|ResetRelativeTransformOnAttach
  ;[1 1 1 0.5 40] Handle attachment
  ;;[1 1 1] Reset Relative Transform On Attach - Set the relative transform back to the one we saved on begin play.
  (Transformation|SetRelativeTransform (CallFunction|GetAttachComponent) (Variables|ProjectileMovement|GetAttachRelativeTrans) false false))

(fn BindPlayerControllerOnUnPosses
  (bind _fortheldobject (Variables|Default|GetFortHeldObject))
  (bind _returnvalue (HeldObjects|Instigator|GetInstigator _fortheldobject))
  (Networking|SwitchHasAuthority
    (:Authority
      (Utilities|IsValid _returnvalue
        (:Is%20Valid
          (bind _asfort_player_pawn_athena (Utilities|Casting|CastToFortPlayerPawnAthena _returnvalue))
          (Game|Player|BindEventtoOnPawnUnpossessed _asfort_player_pawn_athena (EventDispatchers|CreateEvent :!event Server%20Only%20Handle%20Player%20Controller%20OnUnPossess)))))))

(fn BlueprintCanInteract
  (params (InteractingPawn :type Fort%20Pawn%20Object%20Reference) (InteractionBeingAttempted :type EInteractionBeingAttempted%20Enum :default "FirstInteraction") (InteractionType :type TInteractionType%20Enum :default "IT_NoInteraction"))
  (bind _asfort_player_pawn (Utilities|Casting|CastToFortPlayerPawn InteractingPawn))
  (bind _fortheldobject (Variables|Default|GetFortHeldObject))
  ;[1 1 1] Can Interact, If first interact check first interact, if second interact check if the held object can held interact. Also check the blocked tags to make sure we can pick up.
  (Utilities|IsValid InteractingPawn
    (:Is%20Valid
      (if (Utilities|Enum|Equal(Enum) InteractionBeingAttempted "SecondInteraction")
        (Utilities|IsValid _fortheldobject
          (:Is%20Valid
            (bind _returnvalue (HeldObjects|Interaction|CanInteract _fortheldobject _asfort_player_pawn))
            (return (and _returnvalue (not (GameplayTags|HasAnyMatchingGameplayTags _asfort_player_pawn (Variables|Interact|GetTC_BlockPickup))))))
          (:Is%20Not%20Valid
            (return false)))
        (else
          (return (CallFunction|GetCanFirstInteract InteractingPawn)))))))

(fn BlueprintGetInteractionString
  (params (InteractingPawn :type Fort%20Pawn%20Object%20Reference) (InteractionBeingAttempted :type EInteractionBeingAttempted%20Enum :default "FirstInteraction"))
  (return (Utilities|Select :Index InteractionBeingAttempted :FirstInteraction (Variables|Interact|GetFirstInteractString) :SecondInteraction (Variables|Interact|GetSecondInteractString))))

(fn BlueprintGetInteractionTime
  (params (InteractingPawn :type Fort%20Pawn%20Object%20Reference) (InteractionBeingAttempted :type EInteractionBeingAttempted%20Enum :default "FirstInteraction"))
  (return (Utilities|Select :Index InteractionBeingAttempted :FirstInteraction (Variables|Interact|GetFirstInteractTime) :SecondInteraction (Variables|Interact|GetSecondInteractTime) :AllInteraction 0.0) true))

(fn FindAttachAngle
  (params (HitLoc :type Vector))
  ;[1 1 1] Check which is the closes way to the orthagonal options for attachment
  (bind _output_get (Variables|ProjectileMovement|SetMap_DirectionToRelativeVector (Utilities|Map|MakeMap "Forward" (Transformation|GetActorForwardVector) "Backward" (* (Transformation|GetActorForwardVector) -1.0) "Up" (Transformation|GetActorUpVector) "Down" (* (Transformation|GetActorUpVector) -1.0) "Right" (Transformation|GetActorRightVector) "Left" (* (Transformation|GetActorRightVector) -1.0))))
  (bind _keys (Utilities|Map|Keys _output_get))
  (for _array_element _keys
    (if (.ReturnValue (Utilities|Map|Find _output_get _array_element))
      (if (< (Math|Vector|DotProduct (Utilities|Map|Find _output_get _array_element) HitLoc) (LocalVariables|GetCurrentLowest))
        (LocalVariables|SetCurrentLowest (Math|Vector|DotProduct (Utilities|Map|Find _output_get _array_element) HitLoc))
        (Variables|ProjectileMovement|SetDirectionToAttach _array_element)))))

(fn GetAttachComponent
  (return (Variables|Default|GetCapsule)))

(fn GetCanFirstInteract
  (params (InteractingPawn :type Fort%20Pawn%20Object%20Reference))
  (return false))

(fn OnAttachedToVehicle
  (bind _vehicleattached (Variables|Default|GetVehicleAttached))
  (Utilities|FlowControl|Sequence
    (:then_0
      (Networking|SwitchHasAuthority
        (:Authority
          (Utilities|IsValid _vehicleattached
            (:Is%20Valid
              (Game|BindEventtoOnDestroyed _vehicleattached (EventDispatchers|CreateEvent :!event OnVehicleDestroyed)))))))
    (:then_1
      (HeldObjects|TurnOffCollisionSettings (Variables|Default|GetFortHeldObject)))))

(fn OnDetachFromVehicle
  (bind _fortheldobject (Variables|Default|GetFortHeldObject))
  (bind _vehicleattached (Variables|Default|GetVehicleAttached))
  (Utilities|FlowControl|Sequence
    (:then_0
      (Networking|SwitchHasAuthority
        (:Authority
          (Utilities|IsValid _vehicleattached
            (:Is%20Valid
              (Game|UnbindEventfromOnDestroyed _vehicleattached (EventDispatchers|CreateEvent :!event OnVehicleDestroyed))
              (Variables|Default|SetVehicleAttached 0))))))
    (:then_1
      (if (not (HeldObjects|IsHeld _fortheldobject))
        (HeldObjects|TurnonCollisionSettings _fortheldobject)))))

(fn OnRep_AttachTransform
  (bind (_attachtransform_location _attachtransform_rotation) (Math|Transform|BreakTransform (Variables|GetAttachTransform_Location)))
  (bind _attachrelativetrans_location (Math|Transform|BreakTransform (Variables|GetAttachRelativeTrans_Location_X)))
  (bind (_attachrelativetrans_location_x _attachrelativetrans_location_y _attachrelativetrans_location_z) (Math|Vector|BreakVector _attachrelativetrans_location))
  ;[1 1 1] Set the attach point for the base. Calc the relative offset for the base to make sure it stays the same when it rotates. Note that the component needs to have a pivot that is ON CENTER!
  (Transformation|SetWorldTransform (CallFunction|GetAttachComponent) (Math|Transform|MakeTransform (+ _attachtransform_location (Math|Vector|vector+vector (* (Math|Vector|GetForwardVector _attachtransform_rotation) _attachrelativetrans_location_x) (* (Math|Vector|GetRightVector _attachtransform_rotation) _attachrelativetrans_location_y) (* (Math|Vector|GetUpVector _attachtransform_rotation) _attachrelativetrans_location_z))) _attachtransform_rotation (Transformation|GetWorldScale (CallFunction|GetAttachComponent))) false false)
  (CallFunction|Resave))

(fn OnRep_IsAttachedToVehicle
  (if (Variables|Default|GetIsAttachedToVehicle)
    (CallFunction|OnAttachedtoVehicle)
    (else
      (CallFunction|OnDetachfromVehicle))))

(fn OnRep_RepHideActor
  (CallFunction|Resave)
  (if (Variables|Death|GetRepHideActor)
    (CallFunction|HideActor)))

(fn OnRep_RepStartMoving
  (if (== (Variables|ProjectileMovement|GetRepStartMoving) 1)
    (Components|Movement|SetUpdatedComponent (Variables|Default|GetFortProjectileMovement) (Variables|Transformation|GetRootComponent))
    (else
      _))
  (CallFunction|Resave))

(fn ServerOnlyHandleAttachToVehicle
  (params (HitLoc :type Vector) (HitComponentToAttachTo :type Primitive%20Component%20Object%20Reference))
  (bind _hitcomponenttoattachto (Variables|Default|GetHitComponentToAttachTo))
  (Variables|Default|SetHitComponentToAttachTo HitComponentToAttachTo)
  (CallFunction|FindAttachAngle HitLoc)
  (LinkToActorComponent|PerformLinkingCheck (Variables|Default|GetFortLinkToActor) (Variables|ProjectileMovement|GetDirectionToAttach) (Variables|ProjectileMovement|GetEmptyAttachLinkActor))
  (Transformation|AttachActorToComponent _hitcomponenttoattachto "None" "KeepWorld" "KeepWorld" "KeepWorld" false)
  (Collision|SetCollisionResponsetoChannel (Variables|Default|GetCapsule) "ECC_Pawn" "ECR_Ignore") ; Adjust collision so we don't prevent pawns from riding in the vehicle
  (Variables|Default|SetVehicleAttached (Components|GetOwner _hitcomponenttoattachto))
  (Variables|Default|SetIsAttachedToVehicle true))

(fn ServerOnlyHandlePlayerControllerOnUnPossess
  (params (PlayerPawn :type Pawn%20Object%20Reference)))

(fn SetAttachDefaultTransform
  (if (Math|Boolean|ANDBoolean (Networking|IsServer) (Variables|ProjectileMovement|GetAttachToWallsAndCeilings) (not (Variables|ProjectileMovement|GetAlreadySavedDefaultTransform)))
    (Variables|ProjectileMovement|SetAttachRelativeTrans (Transformation|GetRelativeTransform (CallFunction|GetAttachComponent)))
    (Variables|ProjectileMovement|SetAlreadySavedDefaultTransform true)
    (CallFunction|Resave)))

(fn SetBeenThrownPlaced
  ;[1 1 1] If should attach to walls and ceilings then set it has been thrown or placed so that we only start attaching after we have been thrown or placed. Note no need for drop given this is specific to Battle Lab which switching will destroy the object if it has never been placed or thrown.
  (Variables|ProjectileMovement|SetEverBeenThrownPlaced true)
  (CallFunction|Resave))

(fn ShouldDie
  (params (Damage :type Float%20%28single-precision%29) (EventInstigator :type Controller%20Object%20Reference) (DamageCauser :type Actor%20Object%20Reference) (bIsTest :type Boolean) (EffectContext :type Gameplay%20Effect%20Context%20Handle%20Structure%20%28by%20ref%29))
  (return (Variables|Death|GetIsDestructable)))

(fn UbindPlayerControllerOnUnPosses
  (bind _fortheldobject (Variables|Default|GetFortHeldObject))
  (bind _returnvalue (HeldObjects|Instigator|GetInstigator _fortheldobject))
  (Networking|SwitchHasAuthority
    (:Authority
      (Utilities|IsValid _returnvalue
        (:Is%20Valid
          (bind _asfort_player_pawn_athena (Utilities|Casting|CastToFortPlayerPawnAthena _returnvalue))
          (Game|Player|UnbindEventfromOnPawnPossessed _asfort_player_pawn_athena (EventDispatchers|CreateEvent :!event Server%20Only%20Handle%20Player%20Controller%20OnUnPossess)))))))

(fn ConstructionScript)
