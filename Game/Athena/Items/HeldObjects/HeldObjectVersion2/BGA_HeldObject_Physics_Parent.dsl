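; Parent Blueprint for physics-simulated held objects that players pick up, throw, place or drop.
; First interaction is a child-overridable hook; second triggers server pickup through FortHeldObject, gated by blocking tags.
; Lifecycle, hit and buoyancy water-entry events fire replicated gameplay cues, while configurable flags drive owner, instigator and team assignment.
; RepNotifies hide dead objects and wake physics when the linked base is destroyed; objects spawned underwater can optionally self-expire.
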(event EventBeginPlay
  (bind (_returnvalue_x _returnvalue_y _returnvalue_z) (Math|Vector|BreakVector (Transformation|GetActorLocation)))
  ;[1 1 1] Begin Play - Set attach default transform and save off collision setting
  (|CallParentFunction)
  (Networking|FlushNetDormancy)
  (Variables|Default|SetCapsuleDefaultCollisionResponseToPawn (Collision|GetCollisionResponsetoChannel (Variables|Default|GetParentMesh) "ECC_Pawn"))
  ;[1 1 1] Create Water Binding
  (bind _returnvalue (BuildingActor|PhysicsSimulationSupport|GetPhysicsObjectComponent self))
  (Utilities|IsValid (Variables|/Script/FortniteGame.FortPhysicsObjectComponent|GetBuoyancyComponent _returnvalue)
    (:Is%20Valid
      (Cosmetic|BindEventtoOnEnteredWaterDelegate (Variables|/Script/FortniteGame.FortPhysicsObjectComponent|GetBuoyancyComponent _returnvalue) (EventDispatchers|CreateEvent :!event Custom|OnEnterWater)))
    (:Is%20Not%20Valid
      _))
  ;[1 1 1] Wait for Playlist ready, check if we want to kill this object if its underwater. If so, check that it IS underwater and then set lifespan
  (Class|FortAsyncActionCurrentPlaylistReady|PlaylistDataReadyAsync
    (:OnReady
      (if (Variables|Default|GetShouldKillWhenSpawningUnderwater)
        (if (or (Fortnite|WaterHelpers|IsActorSubmergedinOcean self) (<= _returnvalue_z (Fortnite|WaterHelpers|GetOceanTotalHeight)))
          (Actor|SetLifeSpan (Math|Random|RandomFloatinRange 1.0 1e+01)))))))

(event Custom|HideActor
  ;[1 1 1] Hide Actor - Custom Event that can be called on children to do custom changes for when the actor is hidden with the rep notify.
  (Networking|FlushNetDormancy)
  (Rendering|SetActorHiddenInGame true))

(event Interaction|EventBlueprintOnInteract
  (params (InteractingPawn :type Fort%20Pawn%20Object%20Reference) (InteractionBeingAttempted :type EInteractionBeingAttempted%20Enum :default "FirstInteraction"))
  ;[1 1 1] On Interact - Check if the interact is first or second interaction. Execute events that can be overridden in children and by default pick up the item on second interact.
  (if (Variables|Interact|GetRestoreDefaultPawnCollisionOnInteract)
    (Collision|SetCollisionResponsetoChannel (Variables|Default|GetParentMesh) "ECC_Pawn" (Variables|Default|GetCapsuleDefaultCollisionResponseToPawn)) ; Adjust collision so we don't prevent pawns from riding in the vehicle
    (else
      _))
  (Networking|FlushNetDormancy)
  (switch Utilities|FlowControl|Switch|SwitchonEInteractionBeingAttempted InteractionBeingAttempted
    (:FirstInteraction
      (CallFunction|OnFirstInteract InteractingPawn))
    (:SecondInteraction
      (CallFunction|OnSecondInteract InteractingPawn)
      (bind _asfort_player_pawn (Utilities|Casting|CastToFortPlayerPawn InteractingPawn))
      (CallFunction|PrePickupSetup _asfort_player_pawn))))

(event Custom|OnFirstInteract
  (params (Interacting%20Pawn :type Fort%20Pawn%20Object%20Reference)))

(event Custom|OnSecondInteract
  (params (InteractingPawn :type Fort%20Pawn%20Object%20Reference)))

(event Custom|PrePickupSetup
  (params (PawnRef :type Fort%20Player%20Pawn%20Object%20Reference))
  ;[1 1 1] On Interact - Check if the interact is first or second interaction. Execute events that can be overridden in children and by default pick up the item on second interact.
  (if (Networking|IsServer)
    (HeldObjects|Behavior|PickupHeldObject (Variables|Default|GetFortHeldObject) PawnRef)))

(event OnLinkedActorDestroyed(FortLinkToActor)
  (params (DamageCauser :type Actor%20Object%20Reference))
  ;[1 1 1] On Linked Actor Destroyed - Reinit movement when base is destroyed by incrementing a rep notify that sets tells the projectile movement component to begin simulating
  (Networking|FlushNetDormancy)
  (Variables|ProjectileMovement|SetRepStartMoving 1))

(event OnHeldObjectPlaced(FortHeldObject)
  ;[1 1 1] On Held Object Placed - Check that the object has been placed so that attachment on walls works correctly
  (Networking|FlushNetDormancy)
  (CallFunction|SetBeenThrownPlaced)
  (if (or (Variables|Ownership|GetClearOwnerOnAnyDrop) (Variables|Ownership|GetClearOwnerOnPlace))
    (CallFunction|SetOwnerandInstigator 0)))

(event EventDestroyed
  ;[1 1 1] Destroyed - If destroy event is called and we have never been thrown or placed play generic death. This is for things like swapping items on first select
  (Networking|FlushNetDormancy)
  (if (not (Variables|ProjectileMovement|GetEverBeenThrownPlaced))
    (CallFunction|PlayGenericDeath)))

(event Collision|EventHit
  (params (MyComp :type Primitive%20Component%20Object%20Reference) (Other :type Actor%20Object%20Reference) (OtherComp :type Primitive%20Component%20Object%20Reference) (bSelfMoved :type Boolean) (HitLocation :type Vector :default "0, 0, 0") (HitNormal :type Vector :default "0, 0, 0") (NormalImpulse :type Vector :default "0, 0, 0") (Hit :type Hit%20Result%20Structure%20%28by%20ref%29))
  (bind _objectweights (Variables|GameplayCues|GetObjectWeights))
  (bind _returnvalue (Math|Conversions|ToInteger(Byte) _objectweights))
  (bind _returnvalue_2 (Ability|GameplayCue|MakeGameplayCueParameters :NormalizedMagnitude 0.0 :RawMagnitude 0.0 :Location HitLocation :Normal HitNormal :GameplayEffectLevel 1 :AbilityLevel _returnvalue :bReplicateLocationWhenUsingMinimalRepProxy false))
  ;[1 1 1 0.5 40] Execute Gameplay Cues
  ;;[1 1 1] Hit - Send Cues on hit events for the world and player
  (Networking|FlushNetDormancy)
  ;;;[1 1 1] Make sure not in water
  (if (not (CallFunction|IsinWater))
    ;;;[1 1 1] Make sure we are moving some
    (if (>= (Math|Vector|VectorLength (Transformation|GetVelocity)) (Variables|Default|GetHit%20Velocity%20Threshold))
      (Utilities|Casting|CastToFortPawn Other
        (:then
          (Ability|ExecuteGameplayCueWithParamsReplicated (Variables|GameplayCues|GetGC_HitPlayer) _returnvalue_2))
        (:CastFailed
          (Ability|ExecuteGameplayCueWithParamsReplicated (Variables|GameplayCues|GetGC_HitWorld) _returnvalue_2))))))

(event OnHeldObjectThrown(FortHeldObject)
  ;[1 1 1 0.5 40] Execute Gameplay Cues
  ;;[1 1 1] On Held Object Thrown - Send cue for Throw
  (Ability|ExecuteGameplayCueWithParamsReplicated (Variables|GameplayCues|GetGC_Throw) (Ability|GameplayCue|MakeGameplayCueParameters :NormalizedMagnitude 0.0 :RawMagnitude 0.0 :Location (Transformation|GetActorLocation) :Normal (Transformation|GetActorForwardVector) :GameplayEffectLevel 1 :AbilityLevel (Math|Conversions|ToInteger(Byte) (Variables|GameplayCues|GetObjectWeights)) :bReplicateLocationWhenUsingMinimalRepProxy false))
  (CallFunction|SetBeenThrownPlaced)
  (Quest|SendComplexCustomStatEvent (Quests|GetQuestManager (Fortnite|GetFortPlayerControllerfromActor (HeldObjects|Instigator|GetInstigator (Variables|Default|GetFortHeldObject))) "Invalid") 0 "(GameplayTags=)" (Variables|Default|GetThrownQuestCreditTargetTag) 1)
  (if (or (Variables|Ownership|GetClearOwnerOnAnyDrop) (Variables|Ownership|GetClearOwnerOnThrow))
    (CallFunction|SetOwnerandInstigator 0)))

(event Custom|PlayGenericDeath
  ;[1 1 1 0.5 40] Execute Gameplay Cues
  ;;[1 1 1] Play Generic Death - Plays generic death GC used for things that explode on death but should play this death instead when killing from picking it up and breaking it.
  (Ability|ExecuteGameplayCueWithParamsReplicated (Variables|GameplayCues|GetGC_GenericDeath) (Ability|GameplayCue|MakeGameplayCueParameters :NormalizedMagnitude 0.0 :RawMagnitude 0.0 :Location (Transformation|GetActorLocation) :Normal (Transformation|GetActorForwardVector) :GameplayEffectLevel 1 :AbilityLevel (Math|Conversions|ToInteger(Byte) (Variables|GameplayCues|GetObjectWeights)) :bReplicateLocationWhenUsingMinimalRepProxy false))
  (CallFunction|HideActor)
  (if (Networking|IsServer)
    (Variables|Death|SetRepHideActor true)))

(event Custom|OnEnterWater
  (params (Pontoon :type Spherical%20Pontoon%20Structure%20%28by%20ref%29))
  ;[1 1 1 0.5 40] Execute Gameplay Cues
  ;;[1 1 1] On Enter Water - Play Splash Cues
  (bind _returnvalue (Fortnite|HeldObjects|GetHeldObjectAttachedtoPlayer (HeldObjects|Instigator|GetInstigator (Variables|Default|GetFortHeldObject))))
  (if (not (Utilities|Equal(Object) _returnvalue self))
    (bind _returnvalue_2 (BuildingActor|PhysicsSimulationSupport|GetPhysicsObjectComponent self))
    (Utilities|IsValid (Variables|/Script/FortniteGame.FortPhysicsObjectComponent|GetBuoyancyComponent _returnvalue_2)
      (:Is%20Valid
        (bind (_outwaterplanelocation _outwaterplanenormal _outwatersurfaceposition _outwaterdepth _outwaterbodyidx) (Buoyancy|GetLastWaterSurfaceInfo (Variables|/Script/FortniteGame.FortPhysicsObjectComponent|GetBuoyancyComponent _returnvalue_2)))
        (Networking|ForceNetUpdate)
        (Ability|ExecuteGameplayCueWithParamsReplicated (Variables|GameplayCues|GetGC_EnterWater) (Ability|GameplayCue|MakeGameplayCueParameters :NormalizedMagnitude (Math|Conversions|ToFloat(Integer) _outwaterbodyidx) :RawMagnitude _outwaterdepth :Location _outwaterplanelocation :Normal "0,0,0" :GameplayEffectLevel 1 :AbilityLevel (Math|Conversions|ToInteger(Byte) (Variables|GameplayCues|GetObjectWeights)) :bReplicateLocationWhenUsingMinimalRepProxy false))))))

(event EventOnDeathServer
  (params (Damage :type Float%20%28single-precision%29) (DamageTags :type Gameplay%20Tag%20Container%20Structure%20%28by%20ref%29) (Momentum :type Vector :default "0, 0, 0") (HitInfo :type Hit%20Result%20Structure%20%28by%20ref%29) (InstigatedBy :type Controller%20Object%20Reference) (DamageCauser :type Actor%20Object%20Reference) (EffectContext :type Gameplay%20Effect%20Context%20Handle%20Structure))
  ;[1 1 1 0.5 40] Execute Gameplay Cues
  ;;[1 1 1] On Death Server - Play Death GC and then hide actor if settings have not been overridden.
  (Networking|FlushNetDormancy)
  (if (Variables|GameplayCues|GetPlayDeathGC)
    (Ability|ExecuteGameplayCueWithParamsReplicated (Variables|GameplayCues|GetGC_Death) (Ability|GameplayCue|MakeGameplayCueParameters :NormalizedMagnitude 0.0 :RawMagnitude 0.0 :Location (Transformation|GetActorLocation) :Normal (Transformation|GetActorForwardVector) :GameplayEffectLevel 1 :AbilityLevel (Math|Conversions|ToInteger(Byte) (Variables|GameplayCues|GetObjectWeights)) :bReplicateLocationWhenUsingMinimalRepProxy false))
    (if (Variables|Death|GetSetHideActorOnDeath)
      (Variables|Death|SetRepHideActor true))))

(event OnHeldObjectPickedUp(FortHeldObject)
  ;[1 1 1 0.5 40] Execute Gameplay Cues
  ;;[1 1 1] On Pickup - call the reset relative transform event
  (Networking|FlushNetDormancy)
  (if (Variables|ProjectileMovement|GetEverBeenThrownPlaced)
    (Ability|ExecuteGameplayCueWithParamsReplicated (Variables|GameplayCues|GetGC_Pickup) (Ability|GameplayCue|MakeGameplayCueParameters :NormalizedMagnitude 0.0 :RawMagnitude 0.0 :Location (Transformation|GetActorLocation) :Normal (Transformation|GetActorForwardVector) :GameplayEffectLevel 1 :AbilityLevel (Math|Conversions|ToInteger(Byte) (Variables|GameplayCues|GetObjectWeights)) :bReplicateLocationWhenUsingMinimalRepProxy false))))

(event OnHeldObjectAttached(FortHeldObject)
  ;[1 1 1] On Attached - if desired, set the player that picked this object up as its actor owner. Helpful for setting effect causers, team affiliation, etc.
  (if (Variables|Ownership|GetSetOwnerOnPickup)
    (bind _returnvalue (Fortnite|HeldObjects|GetHeldObjectsOwningPlayer self))
    (CallFunction|SetOwnerandInstigator _returnvalue)))

(event EventOnDamageServer
  (params (Damage :type Float%20%28single-precision%29) (DamageTags :type Gameplay%20Tag%20Container%20Structure%20%28by%20ref%29) (Momentum :type Vector :default "0, 0, 0") (HitInfo :type Hit%20Result%20Structure%20%28by%20ref%29) (InstigatedBy :type Controller%20Object%20Reference) (DamageCauser :type Actor%20Object%20Reference) (EffectContext :type Gameplay%20Effect%20Context%20Handle%20Structure))
  ;[1 1 1] On Damage Server - if desired, set the damager as owner. Optionally, only take the first damager as owner
  (if (Variables|Ownership|GetSetOwnerOnDamage)
    (if (Variables|Ownership|GetOnlyFirstDamagerAsOwner)
      (if (not (Variables|Ownership|GetOwnerSetFromDamage))
        _)
      (else
        _))
    (CallFunction|SetOwnerandInstigator (Pawn|GetControlledPawn InstigatedBy))
    (Variables|Ownership|SetOwnerSetFromDamage true)))

(event OnHeldObjectDropped(FortHeldObject)
  ;[1 1 1] On Dropped
  ;;[1 1 1] Check to clear owner
  (if (or (Variables|Ownership|GetClearOwnerOnAnyDrop) (Variables|Ownership|GetClearOwnerOnDrop))
    (CallFunction|SetOwnerandInstigator 0)))

(fn BlueprintCanInteract
  (params (InteractingPawn :type Fort%20Pawn%20Object%20Reference) (InteractionBeingAttempted :type EInteractionBeingAttempted%20Enum :default "FirstInteraction") (InteractionType :type TInteractionType%20Enum :default "IT_NoInteraction"))
  (bind _asfort_player_pawn (Utilities|Casting|CastToFortPlayerPawn InteractingPawn))
  (bind _fortheldobject (Variables|Default|GetFortHeldObject))
  ;[1 1 1] Can Interact, If first interact check first interact, if second interact check if the held object can held interact. Also check the blocked tags to make sure we can pick up.
  (Utilities|IsValid InteractingPawn
    (:Is%20Valid
      (if (Utilities|Enum|Equal(Enum) InteractionBeingAttempted "SecondInteraction")
        (Utilities|IsValid _fortheldobject
          (:Is%20Valid
            (bind _returnvalue (HeldObjects|Interaction|CanInteract _fortheldobject _asfort_player_pawn))
            (return (and _returnvalue (not (GameplayTags|HasAnyMatchingGameplayTags _asfort_player_pawn (Variables|Interact|GetTC_BlockPickup))))))
          (:Is%20Not%20Valid
            (return false)))
        (else
          (return (CallFunction|GetCanFirstInteract InteractingPawn)))))))

(fn BlueprintGetInteractionString
  (params (InteractingPawn :type Fort%20Pawn%20Object%20Reference) (InteractionBeingAttempted :type EInteractionBeingAttempted%20Enum :default "FirstInteraction"))
  (return (Utilities|Select :Index InteractionBeingAttempted :FirstInteraction (Variables|Interact|GetFirstInteractString) :SecondInteraction (Variables|Interact|GetSecondInteractString))))

(fn BlueprintGetInteractionTime
  (params (InteractingPawn :type Fort%20Pawn%20Object%20Reference) (InteractionBeingAttempted :type EInteractionBeingAttempted%20Enum :default "FirstInteraction"))
  (return (Utilities|Select :Index InteractionBeingAttempted :FirstInteraction (Variables|Interact|GetFirstInteractTime) :SecondInteraction (Variables|Interact|GetSecondInteractTime) :AllInteraction 0.0) true))

(fn GetLastInstigator
  (return 0))

(fn GetAttachComponent
  (return (Variables|Default|GetParentMesh)))

(fn GetCanFirstInteract
  (params (InteractingPawn :type Fort%20Pawn%20Object%20Reference))
  (return false))

(fn IsInWater
  (bind _returnvalue (BuildingActor|PhysicsSimulationSupport|GetPhysicsObjectComponent self))
  (Utilities|IsValid (Variables|/Script/FortniteGame.FortPhysicsObjectComponent|GetBuoyancyComponent _returnvalue)
    (:Is%20Valid
      (return (Buoyancy|IsinWaterBody (Variables|/Script/FortniteGame.FortPhysicsObjectComponent|GetBuoyancyComponent _returnvalue))))
    (:Is%20Not%20Valid
      (return false))))

(fn OnRep_RepHideActor
  (CallFunction|Resave)
  (if (Variables|Death|GetRepHideActor)
    (CallFunction|HideActor)))

(fn OnRep_RepStartMoving
  (if (== (Variables|ProjectileMovement|GetRepStartMoving) 1)
    (FortPhysicsSimulationLibrary|WakeUp self "(MaxSimSizeAllowed=Large,bRejectEmbeddedInLandscape=False)")))

(fn SetBeenThrownPlaced
  ;[1 1 1] If should attach to walls and ceilings then set it has been thrown or placed so that we only start attaching after we have been thrown or placed. Note no need for drop given this is specific to Battle Lab which switching will destroy the object if it has never been placed or thrown.
  (Variables|ProjectileMovement|SetEverBeenThrownPlaced true)
  (CallFunction|Resave))

(fn SetOwnerAndInstigator
  (params (Instigator :type Pawn%20Object%20Reference))
  ;[1 1 1] Checking for validity after setting owner and instigator so this function can be used to "clear" them by setting actor ref to null; checking before getting controller to avoid null reference exceptions
  (Actor|SetOwner Instigator)
  (Variables|Actor|SetInstigator Instigator)
  (Utilities|IsValid Instigator
    (:Is%20Valid
      ;[1 1 1] If the new owner has a valid controller, set the team to match. Otherwise, clear team
      (if (Variables|Ownership|GetSetTeamFromOwner)
        (bind _asplayer_controller (Utilities|Casting|CastToPlayerController (Pawn|GetController Instigator))
          (:then
            (CallFunction|SetTeamfromController _asplayer_controller))
          (:CastFailed
            _))
        (else
          _))
      (CallFunction|SetTeam 0))
    (:Is%20Not%20Valid
      _)))

(fn SetTeamFromController
  (params (Controller :type Player%20Controller%20Object%20Reference))
  (CallFunction|SetTeam (Team|GetActorTeam Controller)))

(fn ShouldDie
  (params (Damage :type Float%20%28single-precision%29) (EventInstigator :type Controller%20Object%20Reference) (DamageCauser :type Actor%20Object%20Reference) (bIsTest :type Boolean) (EffectContext :type Gameplay%20Effect%20Context%20Handle%20Structure%20%28by%20ref%29))
  (return (Variables|Death|GetIsDestructable)))

(fn ConstructionScript)
