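; Passive gameplay ability active while a player holds a held object in Fortnite, with blocking tags that stop certain actions.
; On activation it applies a status-tag effect, then drops the object on a force-drop event, or after a short timing delay when the dance-stunned tag is added.
; On end it removes that effect, drops the object if cancelled, and broadcasts a holster-weapon event with false.
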(event Custom|DropHeldObject
  ;[1 1 1] Drop the Held Object - Drop the Held Object. Need to call this on the server!
  (bind _asfort_player_pawn (Utilities|Casting|CastToFortPlayerPawn (Ability|GetAvatarActorfromActorInfo)))
  (bind _returnvalue (Fortnite|HeldObjects|GetHeldObjectComponentAttachedtoPlayer _asfort_player_pawn))
  (Utilities|IsValid _returnvalue
    (:Is%20Valid
      (HeldObjects|Behavior|DropHeldObject _returnvalue))))

(event Ability|EventActivateAbility
  ;;[1 1 1] Apply the GE that gives the player the held object status tag on begin play
  ;[1 0.6563 0] Note that this GA has blocking tags to stop specific actions
  (Ability|ApplyGameplayEffectToOwner (Variables|Default|GetGE_HeldObjectPassive))
  ;[1 1 1] if Danced Stunned Drop.
  (Ability|Tasks|WaitGameplayTagAdd (Variables|Default|GetT_DanceStunned)
    (:then
      ;[1 1 1] If the force drop event is sent to the actor drop the held object.
      (Ability|Tasks|WaitGameplayEvent (Variables|Default|GetEvent_ForceDrop)
        (:EventReceived
          (CallFunction|DropHeldObject))))
    (:Added
      (Utilities|FlowControl|Delay (Variables|Default|GetDanceDropDelay)) ; Unfortunately this needs a small delay to have it not look incorrect timing wise. Otherwise the object drops early and hands stay floating for a moment.
      (CallFunction|DropHeldObject))))

(event Ability|EventOnEndAbility
  (params (bWasCancelled :type Boolean))
  ;[1 1 1] OnEndAbility
  (Utilities|FlowControl|Sequence
    (:then_0
      ;;[0 0 0] Remove the GE with the status tag
      (Ability|RemoveGameplayEffectFromOwnerWithGrantedTags (Variables|Default|GetT_StatusHeldObject))
      (if bWasCancelled
        ;;[0 0 0] If the ability was cancelled, drop our held object
        (bind _asfort_player_pawn (Utilities|Casting|CastToFortPlayerPawn (Ability|GetAvatarActorfromActorInfo)))
        (bind _returnvalue (Fortnite|HeldObjects|GetHeldObjectComponentAttachedtoPlayer _asfort_player_pawn))
        (Utilities|IsValid _returnvalue
          (:Is%20Valid
            (HeldObjects|Behavior|DropHeldObject _returnvalue)))))
    (:then_1
      (Weapon|BroadcastHolsterWeaponEvent (Ability|GetActivatingPawn) false))))
