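; Gameplay ability that throws the player's held object, with windup effect, optional camera mode and trajectory preview.
; It throws on press or primary release; secondary mode cancels on release and awaits primary input.
; The montage trigger leads to a server-side commit and ThrowHeldObject call, with clients requesting it by RPC.
; Trajectory splines use the held-object component's physics values; the socketed indicator is never chosen because HitValidSocket always evaluates false.
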(event Ability|EventActivateAbility
  (bind _cookoverridecamera (Variables|Default|GetCookOverrideCamera))
  (ParentDefaults|PrethrowSetup) ; Ability is instanced per actor, so we need to reset variables\nevery time the ability is activated.
  (if (Variables|AbilityState|GetThrowOnPress)
    (Ability|Tasks|WaitNetSync "OnlyClientWait"
      (:OnSync
        (CallFunction|Throw)))
    (else
      (TrajectoryInterface|SpawnTrajectoryIndicator false :!class /Script/FortniteGame.FortTrajectoryAbilityInterface)
      (Utilities|FlowControl|Sequence
        (:then_0
          ;[1 1 1] Add camera mode if valid
          (if (Utilities|IsValidClass _cookoverridecamera)
            (Camera|PushCameraMode _cookoverridecamera)))
        (:then_1
          ;[1 1 1] If triggered by secondary fire, wait for primary input to be pressed to throw
          (if (Variables|Default|GetIsSecondaryAbility)
            (Ability|SetShouldBlockOtherAbilities true)
            (if (TouchPlatform|IsUsingTouch (Ability|GetPlayerControllerfromActorInfo))
              (Ability|Tasks|WaitGameplayEvent "(TagName=\"Event.TouchConsumableThrown\")" 0 true
                (:then
                  (if (Networking|IsServer)
                    _))
                (:EventReceived
                  _))
              (Ability|Tasks|WaitNetSync "OnlyServerWait"
                (:OnSync
                  (CallFunction|Throw)
                  (CallFunction|RemoveCameraMode)))
              (else
                (Ability|Tasks|WaitGameplayEvent "(TagName=\"Event.Primary.Pressed\")" 0 true
                  (:then
                    _)
                  (:EventReceived
                    _))))))))))

(event TrajectoryInterface|EventInitTrajectoryVariables
  ;[1 1 1] Spawn a dummy projectile to check the projectile's max speed, gravity scale, and sphere extent. Used by the trajectory display.
  (Utilities|IsValid (Variables|Default|GetHeldObjectComponent)
    (:Is%20Valid
      (Variables|TrajectoryDisplay|SetDummyMaxSpeed (HeldObjects|Trajectory|GetMaxSpeed (Variables|Default|GetHeldObjectComponent)))
      (Variables|TrajectoryDisplay|SetDummyGravity (HeldObjects|Trajectory|GetGravityZ (Variables|Default|GetHeldObjectComponent)))
      (Variables|TrajectoryDisplay|SetDummyExtent (HeldObjects|Trajectory|GetSweptTraceRadius (Variables|Default|GetHeldObjectComponent))))))

(event Ability|EventOnAbilityInputReleased
  (Variables|AbilityState|SetAbilityKeyPressed false)
  ;[1 1 1] Input release is handled differently if the ability is triggered via primary fire or secondary fire. If primary fire, initiate the throw on release. If secondary fire, the throw is cancelled
  (if (Variables|Default|GetIsSecondaryAbility)
    ;;[1 1 1] Secondary Fire - If we aren't already in the process of throwing, cancel the ability
    (if (not (Variables|Default|GetIsThrowing))
      (Development|PrintString "Throw ability cancelled")
      (Ability|EndAbility))
    (elif (Variables|AbilityState|GetInThrowWindup)
      ;;[1 1 1] Primary Fire - Throw the projectile if input is released while still in throw windup.
      (CallFunction|Throw)
      (CallFunction|RemoveCameraMode))))

(event Custom|RemoveCameraMode
  (if (Utilities|IsValidClass (Variables|Default|GetCookOverrideCamera))
    (Camera|PopCameraMode)))

(event Custom|Throw
  (Variables|Default|SetIsThrowing true)
  (CallFunction|RemoveCameraMode)
  ;[1 1 1] BR: ShouldBlockAbilities = true
  ; StW: CanBeCanceled = false,  ShouldBlockAbilities = true
  (Ability|SetShouldBlockOtherAbilities true)
  (ParentDefaults|PostThrowCleanup)
  ;[1 1 1] Projectile is spawned when the montage reaches ability notify
  (Ability|Tasks|PlayMontageandWaitforTrigger "None" (Variables|Default|GetThrowMontage) 1.0 "FirstSection" "None" 1.0 "SimulateOnServer"
    (:Triggered
      (CallFunction|AttemptSpawnThrownProjectile)
      (if (not (Variables|ParentDefaults|GetWaitForEndAbilityToCompleteThrow))
        (HeldObjects|Behavior|OnThrowComplete (Variables|Default|GetHeldObjectComponent)))
      (Ability|SetShouldBlockOtherAbilities false)
      (Ability|EndAbility))
    (:Cancelled
      _)
    (:Completed
      _)))

(event Custom|ServerSpawnThrownProjectile
  (if (Networking|IsServer)
    (CallFunction|Throw)))

(event Custom|AttemptSpawnThrownProjectile
  (Ability|ExecuteGameplayCueOnOwner (Variables|Default|GetGC_OnThrow)) ; GC to play on the player when the projectile is thrown
  (Networking|ForceNetUpdate (CallFunction|GetAvatarasFortPlayerPawn))
  (if (Networking|IsServer)
    ;[1 1 1] Commit the ability here to prevent exploits relating to spawning multiple projectiles
    (bind _returnvalue (Ability|CommitAbility))
    (if _returnvalue
      (Variables|ParentDefaults|SetExplosionEffectContainerSpec (Ability|MakeGameplayEffectContainerSpec (Variables|ParentDefaults|GetEC_DefaultExplosion)))
      (bind (_location _rotation) (ParentDefaults|GetSpawnLocationandRotation))
      (HeldObjects|Behavior|ThrowHeldObject (Variables|Default|GetHeldObjectComponent) _location _rotation)
      (Ability|Tasks|WaitDelay 0.4
        (:OnFinish
          (Ability|SendGameplayEvent "(TagName=\"Event.HeldObject.Released\")")
          (HeldObjects|Behavior|OnThrowComplete (Variables|Default|GetHeldObjectComponent)))) ; Would prefer to end the ability without a delay if possible.
      (else
        (Ability|EndAbility)))
    (else
      ;[1 1 1] If on client, fire RPC to request server to spawn projectile accordingly
      (CallFunction|ServerSpawnThrownProjectile))))

(event TrajectoryInterface|EventUpdateTrajectory
  (TrajectoryDisplay|UpdateTrajectorySpline))

(event Ability|EventOnEndAbility
  (params (bWasCancelled :type Boolean))
  ;[1 1 1] Failsafe to clean up the trajectory when the ability ends
  (Utilities|IsValid (Variables|Default|GetHeldObjectComponent)
    (:Is%20Valid
      (HeldObjects|Behavior|OnThrowComplete (Variables|Default|GetHeldObjectComponent)))
    (:Is%20Not%20Valid
      _))
  (|CallParentFunction bWasCancelled)
  (ParentDefaults|PostThrowCleanup)
  (Variables|Default|SetIsThrowing false))

(event TrajectoryInterface|EventCleanupTrajectoryIndicatorOnUnequip
  (TrajectoryDisplay|CleanupTrajectory))

(fn CleanupTrajectory
  (Class|BPProjectileTrajectory|SetShouldUpdatefromOwner (Variables|TrajectoryDisplay|GetTrajectoryIndicator) false 0)
  (Class|BPProjectileTrajectory|SetShouldUpdatefromOwner (Variables|Default|GetProjectileTrajectorySocketed) false 0)
  (Actor|DestroyActor (Variables|Default|GetProjectileTrajectorySocketed)))

(fn GetOwningPlayer
  (bind (_owneractor _avataractor) (Utilities|Struct|BreakGameplayAbilityActorInfo (Ability|GetActorInfo)))
  (Utilities|IsValid (Variables|ReferencedActors|GetPlayerPawn)
    (:Is%20Valid
      (return (Variables|ReferencedActors|GetPlayerPawn)))
    (:Is%20Not%20Valid
      (bind _asfort_player_pawn (Utilities|Casting|CastToFortPlayerPawn _avataractor)
        (:then
          (return _asfort_player_pawn))
        (:CastFailed
          (return 0))))))

(fn GetProjectileTrajectoryActor
  (return (Variables|TrajectoryDisplay|GetTrajectoryIndicatorClass)))

(fn GetProjectileTrajectoryPoints
  (bind (_location _rotation) (ParentDefaults|GetSpawnLocationandRotation))
  (bind _returnvalue (Fortnite|HeldObjects|GetHeldObjectAttachedtoPlayer (Variables|ReferencedActors|GetPlayerPawn)))
  (LocalVariables|SetLocalHeldObject _returnvalue)
  (bind (_outhitresult _outsplinepoints _outsplinetangents) (Projectile|CalculateProjectileTrajectorySplinewithHitProfile (ParentDefaults|GetOwningPlayer) _location (* (Math|Rotator|GetRotationXVector _rotation) (Variables|TrajectoryDisplay|GetDummyMaxSpeed)) (Variables|TrajectoryDisplay|GetDummyMaxSpeed) (Variables|TrajectoryDisplay|GetDummyGravity) 0.0 0.0 (Variables|TrajectoryDisplay|GetDummyTimeStep) (Variables|TrajectoryDisplay|GetDummyExtent) "FortTrajectoryBlocksWater" (Utilities|Array|MakeArray (LocalVariables|GetLocalHeldObject) (Variables|ReferencedActors|GetPlayerPawn)) 1))
  (return _outsplinepoints _outsplinetangents _outhitresult))

(fn GetSpawnLocationAndRotation
  (return (+ (Transformation|GetActorLocation (ParentDefaults|GetOwningPlayer)) (Math|Vector|RotateVectorAroundAxis (select (Variables|/Script/Engine.Character|GetbIsCrouched (ParentDefaults|GetOwningPlayer)) (Variables|ParentDefaults|GetGrenadeTargetingOriginOffset_Crouched) (Variables|ParentDefaults|GetGrenadeTargetingOriginOffset)) (.yaw (Transformation|GetActorRotation (ParentDefaults|GetOwningPlayer))) "0,0.000000,1.000000")) (Math|Rotator|MakeRotator (.roll (Targeting|GetDefaultTargetingRotation (ParentDefaults|GetOwningPlayer))) (Math|Float|Min(Float) (Math|Rotator|NormalizeAxis (+ (.pitch (Targeting|GetDefaultTargetingRotation (ParentDefaults|GetOwningPlayer))) (Variables|ParentDefaults|GetAdditionalThrowAngle))) (Variables|ParentDefaults|GetMaxTossPitch)) (select (Variables|ParentDefaults|GetMeshRotationForThrowDirection) (+ (.yaw (Transformation|GetWorldRotation (Variables|/Script/Engine.Character|GetMesh (ParentDefaults|GetOwningPlayer)))) 9e+01) (.yaw (Targeting|GetDefaultTargetingRotation (ParentDefaults|GetOwningPlayer)))))))

(fn HitValidSocket
  (params (HitResult :type Hit%20Result%20Structure))
  (bind (_bblockinghit _binitialoverlap _time _distance _location _impactpoint _normal _impactnormal _physmat _hitactor _hitcomponent) (Collision|BreakHitResult HitResult))
  (Utilities|IsValid _hitactor
    (:Is%20Valid
      (bind _asb_held_object_socket (Utilities|Casting|CastToB_HeldObjectSocket _hitactor)
        (:then
          (Utilities|Casting|CastToSphereComponent _hitcomponent
            (:then
              (return (and false (not (Variables|/Game/Athena/Items/HeldObjects/B_HeldObjectSocket.B_HeldObjectSocket_C|GetHasSocketedHeldObject _asb_held_object_socket)))))
            (:CastFailed
              _)))
        (:CastFailed
          _))
      (return false))
    (:Is%20Not%20Valid
      _)))

(fn CanActivateAbility
  (params (ActorInfo :type Gameplay%20Ability%20Actor%20Info%20Structure) (Handle :type Gameplay%20Ability%20Spec%20Handle%20Structure))
  (bind (_owneractor _avataractor) (Utilities|Struct|BreakGameplayAbilityActorInfo ActorInfo))
  (bind _asfort_player_pawn (Utilities|Casting|CastToFortPlayerPawn _avataractor)
    (:then
      (bind _returnvalue (Fortnite|HeldObjects|GetHeldObjectComponentAttachedtoPlayer _asfort_player_pawn))
      (Utilities|IsValid _returnvalue
        (:Is%20Valid
          (return 0 (HeldObjects|IsHeld _returnvalue)))
        (:Is%20Not%20Valid
          _)))
    (:CastFailed
      _))
  (return 0 false))

(fn OnHeldObjectSetup)

(fn OnHeldObjectThrown
  (params (SpawnedProjectile :type Fort%20Projectile%20Base%20Object%20Reference)))

(fn PostThrowCleanup
  ;[1 1 1] Clear the looping trajectory display timer, destroy the trajectory indicator, and remove windup GE.
  (Variables|AbilityState|SetInThrowWindup false)
  (Ability|RemoveGameplayEffectFromOwnerWithHandle (Variables|ParentDefaults|GetHandle_GE_InWindup))
  (if (not (Variables|Default|GetSpawnedOnEquip))
    (TrajectoryDisplay|CleanupTrajectory)))

(fn PrethrowSetup
  (TrajectoryDisplay|SetupSplineActors)
  ;[1 1 1] The input is being held down and the actor is winding up to throw the projectile
  (Variables|AbilityState|SetAbilityKeyPressed true)
  (Variables|AbilityState|SetInThrowWindup true)
  (Variables|Default|SetIsThrowing false)
  ;[1 1 1] Add a GE that applies looping GC while in the windup
  (bind _returnvalue (Ability|ApplyGameplayEffectToOwner (Variables|Default|GetGE_InWindup)))
  (Variables|ParentDefaults|SetHandle_GE_InWindup _returnvalue))

(fn SetupLateTrajectoryData
  (if (or (not (Utilities|IsValid (Variables|Default|GetHeldObject) :!class /Script/Engine.KismetSystemLibrary)) (not (Utilities|IsValid (Variables|Default|GetHeldObjectComponent) :!class /Script/Engine.KismetSystemLibrary)))
    (if (TouchPlatform|IsMobileGame)
      (TrajectoryDisplay|SetupSplineActors)
      (else
        (ParentDefaults|PrethrowSetup)))
    (TrajectoryInterface|InitTrajectoryVariables)))

(fn SetupDummyProjectile
  (bind _trajectoryindicatorclass (Variables|TrajectoryDisplay|GetTrajectoryIndicatorClass))
  (if (Utilities|IsValidClass _trajectoryindicatorclass)
    (bind _returnvalue (Game|SpawnActorfromClass _trajectoryindicatorclass (Transformation|GetActorTransform (ParentDefaults|GetOwningPlayer)) "AlwaysSpawn" "SelectDefaultAtRuntime" 0 (ParentDefaults|GetOwningPlayer)))
    (bind _asbp_projectile_trajectory (Utilities|Casting|CastToBP_ProjectileTrajectory _returnvalue))
    (Variables|TrajectoryDisplay|SetTrajectoryIndicator _asbp_projectile_trajectory)
    (TrajectoryInterface|InitTrajectoryVariables :!class /Script/FortniteGame.FortTrajectoryAbilityInterface)))

(fn SetupDummyProjectileSocketed
  (Utilities|IsValid (Variables|Default|GetProjectileTrajectorySocketed)
    (:Is%20Not%20Valid
      (bind _returnvalue (Game|SpawnActorfromClass (Variables|Default|GetTrajectoryIndicatorSocketedClass) (Transformation|GetActorTransform (ParentDefaults|GetOwningPlayer)) "AlwaysSpawn" "SelectDefaultAtRuntime" 0 0))
      (bind _asbp_projectile_trajectory (Utilities|Casting|CastToBP_ProjectileTrajectory _returnvalue))
      (Variables|Default|SetProjectileTrajectorySocketed _asbp_projectile_trajectory))))

(fn SetupSplineActors
  (bind (_owneractor _avataractor) (Utilities|Struct|BreakGameplayAbilityActorInfo (Ability|GetActorInfo)))
  (bind _asfort_player_pawn (Utilities|Casting|CastToFortPlayerPawn _avataractor)
    (:then
      (bind _output_get (Variables|ReferencedActors|SetPlayerPawn _asfort_player_pawn))
      (bind _returnvalue (Fortnite|HeldObjects|GetHeldObjectAttachedtoPlayer _output_get))
      (Utilities|IsValid _returnvalue
        (:Is%20Valid
          (bind _output_get_2 (Variables|Default|SetHeldObject _returnvalue))
          (Variables|Default|SetHeldObjectComponent (Actor|GetComponentbyClass _output_get_2 "/Script/FortniteGame.FortHeldObjectComponent"))
          ;[1 1 1] We allow a player to cancel out of the windup for most thrown projectiles
          (Ability|SetShouldBlockOtherAbilities false))))
    (:CastFailed
      (Ability|EndAbility))))

(fn ShouldOnlyShowTrajectoryOnUse
  (return false))

(fn SpawnThrownProjectile
  (params (ProjectileClass :type Fort%20Projectile%20Base%20Class%20Reference) (ProjectileSpawnLocation :type Vector%20%28by%20ref%29) (ProjectileSpawnDirection :type Rotator%20%28by%20ref%29) (ExplosionEffectContainerSpec :type Fort%20Gameplay%20Effect%20Container%20Spec%20Structure) (ProjectileInitialSpeed :type Float%20%28double-precision%29) (ProjectileGravityScale :type Float%20%28double-precision%29)))

(fn SpawnTrajectoryIndicator
  (params (bSpawnOnEquip :type Boolean))
  (bind _player_pawn (ParentDefaults|GetOwningPlayer))
  (bind _returnvalue (Pawn|IsLocallyControlled _player_pawn))
  (CallFunction|SetupLateTrajectoryData)
  (CallFunction|SetupDummyProjectileSocketed)
  (Utilities|IsValid (Variables|TrajectoryDisplay|GetTrajectoryIndicator)
    (:Is%20Not%20Valid
      (Variables|Default|SetSpawnedOnEquip bSpawnOnEquip)
      (Utilities|FlowControl|Sequence
        (:then_0
          (if _returnvalue
            (TrajectoryDisplay|SetupDummyProjectile)))
        (:then_1
          ;[1 1 1] Only show trajectory display if input is held down
          (if (and _returnvalue (or (Variables|AbilityState|GetAbilityKeyPressed) (Variables|Default|GetSpawnedOnEquip)))
            (Class|BPProjectileTrajectory|SetShouldUpdatefromOwner (Variables|TrajectoryDisplay|GetTrajectoryIndicator) true self)
            (return (Variables|TrajectoryDisplay|GetTrajectoryIndicator))))))))

(fn UpdateTrajectorySpline
  (bind _trajectoryindicator (Variables|TrajectoryDisplay|GetTrajectoryIndicator))
  (bind _projectiletrajectorysocketed (Variables|Default|GetProjectileTrajectorySocketed))
  (CallFunction|SetupLateTrajectoryData)
  (if (and (Utilities|IsValid (Variables|TrajectoryDisplay|GetTrajectoryIndicator) :!class /Script/Engine.KismetSystemLibrary) (Utilities|IsValid (Variables|Default|GetProjectileTrajectorySocketed) :!class /Script/Engine.KismetSystemLibrary))
    (bind (_outsplinepoints _outsplinetangents _outhitresult) (TrajectoryInterface|GetProjectileTrajectoryPoints :!class /Script/FortniteGame.FortTrajectoryAbilityInterface))
    (bind _cansocket (CallFunction|HitValidSocket _outhitresult))
    (Trajectory|SetTrajectorySpline (select _cansocket _projectiletrajectorysocketed _trajectoryindicator) _outsplinepoints _outsplinetangents)
    (Rendering|SetActorHiddenInGame (select _cansocket _projectiletrajectorysocketed _trajectoryindicator) false)
    (Rendering|SetActorHiddenInGame (select _cansocket (Variables|TrajectoryDisplay|GetTrajectoryIndicator) (Variables|Default|GetProjectileTrajectorySocketed)) true)))
