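; Gameplay ability for throwing a held object, a component-based variant for AI going by its name.
; Activation is allowed only when the avatar is a Fort player pawn with a valid attached held object component.
; Trajectory initialization runs the parent logic, then derives the preview spline's dummy gravity from the component's gravity scale and a fixed downward constant to correct the arc.
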(event TrajectoryInterface|EventInitTrajectoryVariables
  (|CallParentFunction)
  (Variables|TrajectoryDisplay|SetDummyGravity (* (HeldObjects|Trajectory|GetGravityScale (Variables|Default|GetHeldObjectComponent)) -2.8e+03)) ; Fix Gravity Component of Trajectory Spline
)

(fn CanActivateAbility
  (params (ActorInfo :type Gameplay%20Ability%20Actor%20Info%20Structure) (Handle :type Gameplay%20Ability%20Spec%20Handle%20Structure))
  (bind _asfort_player_pawn (Utilities|Casting|CastToFortPlayerPawn (Ability|GetAvatarActorfromActorInfo))
    (:then
      (bind _returnvalue (Fortnite|HeldObjects|GetHeldObjectComponentAttachedtoPlayer _asfort_player_pawn))
      (Utilities|IsValid _returnvalue
        (:Is%20Valid
          (return 0 true))
        (:Is%20Not%20Valid
          _)))
    (:CastFailed
      _))
  (return 0 false))
